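# Animation Blueprint for the ShiverFlame pickaxe that procedurally rattles and pulses its engine pipes.
# Each update builds sine and random pipe rotations and scale from game time, with stronger rotation during melee animations.
# A Modify Bone chain applies these to the input pose, then adds fixed yaw offsets.
# Motor Rotate is unused, two EnginePipe02 right-side nodes are unconnected, and the last node resets EnginePipe01_Scale_r to unit scale.
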AnimBlueprint /BRCosmetics/Weapons/FORT_Melee/Pickaxe_ShiverFlame/Meshes/ShiverFlame_Axe_AnimBP
target_skeleton=&/ShiverFlame_Axe_Skeleton
graphs [
 AnimGraph {nodes [
   "Transform (Modify) Bone - Bone: dyn_EnginePipe02_Pipe03_r"(Alpha=1, Translation="0,0,0", Rotation=Get MainMassRotate() -> (MainMassRotate:Rotator), Scale="1,1,1") {
    rotationMode=BMM_Replace
    rotationSpace=BCS_BoneSpace
    alphaBoolBlend {blendOption=Linear}
    boneToModify {boneName=dyn_EnginePipe02_Pipe03_r}
   }
   "Transform (Modify) Bone - Bone: dyn_EnginePipe02_Pipe04_r"(Alpha=1, Translation="0,0,0", Rotation=Get MainMassRotate(), Scale="1,1,1") {
    rotationMode=BMM_Replace
    rotationSpace=BCS_BoneSpace
    alphaBoolBlend {blendOption=Linear}
    boneToModify {boneName=dyn_EnginePipe02_Pipe04_r}
   }
   AnimGraphNode_SubInput_0 = Input Pose()
   n1 = "Transform (Modify) Bone - Bone: dyn_EnginePipe02_Pipe05_Scale_l"(
    ComponentPose=Local To Component(LocalPose=@AnimGraphNode_SubInput_0)
    Alpha=1
    Translation="0,0,0"
    Rotation="0,0,0"
    Scale=Get Pipe Scale() -> (Pipe Scale:Vector)
   ) {
    scaleMode=BMM_Replace
    scaleSpace=BCS_BoneSpace
    alphaBoolBlend {blendOption=Linear}
    boneToModify {boneName=dyn_EnginePipe02_Pipe05_Scale_l}
   }
   n2 = "Transform (Modify) Bone - Bone: dyn_EnginePipe02_Pipe01_l"(ComponentPose=@n1, Alpha=1, Translation="0,0,0", Rotation=Get PipeRotate() -> (PipeRotate:Rotator), Scale="1,1,1") {
    rotationMode=BMM_Replace
    rotationSpace=BCS_BoneSpace
    alphaBoolBlend {blendOption=Linear}
    alphaScaleBiasClamp {bInterpResult=true}
    boneToModify {boneName=dyn_EnginePipe02_Pipe01_l}
   }
   n3 = "Transform (Modify) Bone - Bone: dyn_EnginePipe02_Pipe02_l"(ComponentPose=@n2, Alpha=1, Translation="0,0,0", Rotation=Get MainMassRotate(), Scale="1,1,1") {
    rotationMode=BMM_Replace
    rotationSpace=BCS_BoneSpace
    alphaBoolBlend {blendOption=Linear}
    boneToModify {boneName=dyn_EnginePipe02_Pipe02_l}
   }
   n4 = "Transform (Modify) Bone - Bone: dyn_EnginePipe02_Pipe03_l"(ComponentPose=@n3, Alpha=1, Translation="0,0,0", Rotation=Get MainMassRotate(), Scale="1,1,1") {
    rotationMode=BMM_Replace
    rotationSpace=BCS_BoneSpace
    alphaBoolBlend {blendOption=Linear}
    boneToModify {boneName=dyn_EnginePipe02_Pipe03_l}
   }
   n5 = "Transform (Modify) Bone - Bone: dyn_EnginePipe02_Pipe04_l"(ComponentPose=@n4, Alpha=1, Translation="0,0,0", Rotation=Get MainMassRotate(), Scale="1,1,1") {
    rotationMode=BMM_Replace
    rotationSpace=BCS_BoneSpace
    alphaBoolBlend {blendOption=Linear}
    boneToModify {boneName=dyn_EnginePipe02_Pipe04_l}
   }
   n6 = "Transform (Modify) Bone - Bone: dyn_EnginePipe02_Pipe05_Scale_r"(ComponentPose=@n5, Alpha=1, Translation="0,0,0", Rotation="0,0,0", Scale=Get Pipe Scale()) {
    scaleMode=BMM_Replace
    scaleSpace=BCS_BoneSpace
    alphaBoolBlend {blendOption=Linear}
    boneToModify {boneName=dyn_EnginePipe02_Pipe05_Scale_r}
   }
   n7 = "Transform (Modify) Bone - Bone: dyn_EnginePipe02_Pipe01_r"(ComponentPose=@n6, Alpha=1, Translation="0,0,0", Rotation=Get PipeRotate(), Scale="1,1,1") {
    rotationMode=BMM_Replace
    rotationSpace=BCS_BoneSpace
    alphaBoolBlend {blendOption=Linear}
    alphaScaleBiasClamp {bInterpResult=true}
    boneToModify {boneName=dyn_EnginePipe02_Pipe01_r}
   }
   n8 = "Transform (Modify) Bone - Bone: dyn_EnginePipe02_Pipe02_r"(ComponentPose=@n7, Alpha=1, Translation="0,0,0", Rotation=Get MainMassRotate(), Scale="1,1,1") {
    rotationMode=BMM_Replace
    rotationSpace=BCS_BoneSpace
    alphaBoolBlend {blendOption=Linear}
    boneToModify {boneName=dyn_EnginePipe02_Pipe02_r}
   }
   n9 = "Transform (Modify) Bone - Bone: dyn_EnginePipe01_Scale_l"(ComponentPose=@n8, Alpha=1, Translation="0,0,0", Rotation="0,0,0", Scale=Get Pipe Scale()) {scaleMode=BMM_Replace, scaleSpace=BCS_BoneSpace, alphaBoolBlend {blendOption=Linear}, boneToModify {boneName=dyn_EnginePipe01_Scale_l}}
   n10 = "Transform (Modify) Bone - Bone: dyn_EnginePipe01_Pipe01_l"(ComponentPose=@n9, Alpha=1, Translation="0,0,0", Rotation=Get PipeRotate_Small() -> (PipeRotate_Small:Rotator), Scale="1,1,1") {
    rotationMode=BMM_Replace
    rotationSpace=BCS_BoneSpace
    alphaBoolBlend {blendOption=Linear}
    alphaScaleBiasClamp {bInterpResult=true}
    boneToModify {boneName=dyn_EnginePipe01_Pipe01_l}
   }
   n11 = "Transform (Modify) Bone - Bone: dyn_EnginePipe01_Pipe02_l"(ComponentPose=@n10, Alpha=1, Translation="0,0,0", Rotation=Get MainMassRotate(), Scale="1,1,1") {
    rotationMode=BMM_Replace
    rotationSpace=BCS_BoneSpace
    alphaBoolBlend {blendOption=Linear}
    boneToModify {boneName=dyn_EnginePipe01_Pipe02_l}
   }
   n12 = "Transform (Modify) Bone - Bone: dyn_EnginePipe01_Pipe03_l"(ComponentPose=@n11, Alpha=1, Translation="0,0,0", Rotation=Get MainMassRotate(), Scale="1,1,1") {
    rotationMode=BMM_Replace
    rotationSpace=BCS_BoneSpace
    alphaBoolBlend {blendOption=Linear}
    boneToModify {boneName=dyn_EnginePipe01_Pipe03_l}
   }
   n13 = "Transform (Modify) Bone - Bone: dyn_EnginePipe01_Pipe04_l"(ComponentPose=@n12, Alpha=1, Translation="0,0,0", Rotation=Get MainMassRotate(), Scale="1,1,1") {
    rotationMode=BMM_Replace
    rotationSpace=BCS_BoneSpace
    alphaBoolBlend {blendOption=Linear}
    boneToModify {boneName=dyn_EnginePipe01_Pipe04_l}
   }
   n14 = "Transform (Modify) Bone - Bone: dyn_EnginePipe01_Scale_r"(ComponentPose=@n13, Alpha=1, Translation="0,0,0", Rotation="0,0,0", Scale=Get Pipe Scale()) {scaleMode=BMM_Replace, scaleSpace=BCS_BoneSpace, alphaBoolBlend {blendOption=Linear}, boneToModify {boneName=dyn_EnginePipe01_Scale_r}}
   n15 = "Transform (Modify) Bone - Bone: dyn_EnginePipe01_Pipe01_r"(ComponentPose=@n14, Alpha=1, Translation="0,0,0", Rotation=Get PipeRotate_Small(), Scale="1,1,1") {
    rotationMode=BMM_Replace
    rotationSpace=BCS_BoneSpace
    alphaBoolBlend {blendOption=Linear}
    alphaScaleBiasClamp {bInterpResult=true}
    boneToModify {boneName=dyn_EnginePipe01_Pipe01_r}
   }
   n16 = "Transform (Modify) Bone - Bone: dyn_EnginePipe01_Pipe02_r"(ComponentPose=@n15, Alpha=1, Translation="0,0,0", Rotation=Get MainMassRotate(), Scale="1,1,1") {
    rotationMode=BMM_Replace
    rotationSpace=BCS_BoneSpace
    alphaBoolBlend {blendOption=Linear}
    boneToModify {boneName=dyn_EnginePipe01_Pipe02_r}
   }
   n17 = "Transform (Modify) Bone - Bone: dyn_EnginePipe01_Pipe03_r"(ComponentPose=@n16, Alpha=1, Translation="0,0,0", Rotation=Get MainMassRotate(), Scale="1,1,1") {
    rotationMode=BMM_Replace
    rotationSpace=BCS_BoneSpace
    alphaBoolBlend {blendOption=Linear}
    boneToModify {boneName=dyn_EnginePipe01_Pipe03_r}
   }
   n18 = "Transform (Modify) Bone - Bone: dyn_EnginePipe01_Pipe04_r"(ComponentPose=@n17, Alpha=1, Translation="0,0,0", Rotation=Get MainMassRotate(), Scale="1,1,1") {
    rotationMode=BMM_Replace
    rotationSpace=BCS_BoneSpace
    alphaBoolBlend {blendOption=Linear}
    boneToModify {boneName=dyn_EnginePipe01_Pipe04_r}
   }
   n19 = "Transform (Modify) Bone - Bone: dyn_EnginePipe02_Pipe01_l"(ComponentPose=@n18, Alpha=1, Translation="0,0,0", Rotation="0,3,0", Scale="1,1,1") {
    rotationMode=BMM_Additive
    scaleMode=BMM_Replace
    scaleSpace=BCS_BoneSpace
    alphaBoolBlend {blendOption=Linear}
    boneToModify {boneName=dyn_EnginePipe02_Pipe01_l}
    rotation {yaw=3}
   }
   n20 = "Transform (Modify) Bone - Bone: dyn_EnginePipe02_Pipe01_r"(ComponentPose=@n19, Alpha=1, Translation="0,0,0", Rotation="0,-4,0", Scale="1,1,1") {
    rotationMode=BMM_Additive
    scaleMode=BMM_Replace
    scaleSpace=BCS_BoneSpace
    alphaBoolBlend {blendOption=Linear}
    boneToModify {boneName=dyn_EnginePipe02_Pipe01_r}
    rotation {yaw=-4}
   }
   n21 = "Transform (Modify) Bone - Bone: dyn_EnginePipe01_Pipe01_l"(ComponentPose=@n20, Alpha=1, Translation="0,0,0", Rotation="0,1,0", Scale="1,1,1") {
    rotationMode=BMM_Additive
    scaleMode=BMM_Replace
    scaleSpace=BCS_BoneSpace
    alphaBoolBlend {blendOption=Linear}
    boneToModify {boneName=dyn_EnginePipe01_Pipe01_l}
    rotation {yaw=1}
   }
   n22 = "Transform (Modify) Bone - Bone: dyn_EnginePipe01_Pipe01_r"(ComponentPose=@n21, Alpha=1, Translation="0,0,0", Rotation="0,-1,0", Scale="1,1,1") {
    rotationMode=BMM_Additive
    scaleMode=BMM_Replace
    scaleSpace=BCS_BoneSpace
    alphaBoolBlend {blendOption=Linear}
    boneToModify {boneName=dyn_EnginePipe01_Pipe01_r}
    rotation {yaw=-1}
   }
   n23 = "Transform (Modify) Bone - Bone: dyn_EnginePipe01_Scale_r"(ComponentPose=@n22, Alpha=1, Translation="0,0,0", Rotation="0,0,0", Scale="1,1,1") {scaleMode=BMM_Replace, scaleSpace=BCS_BoneSpace, alphaBoolBlend {blendOption=Linear}, boneToModify {boneName=dyn_EnginePipe01_Scale_r}}
   Output Pose(Result=Component To Local(ComponentPose=@n23)) {layerGroup=None}
  ]}
 EventGraph {nodes [
   Event BlueprintUpdateAnimation()
   n1 = "Sin (Radians)"(A=float * float(A=float * float(A=1, B=20), B=GetGameTimeInSeconds()))
   n2 = RandomFloatInRange(
    Min=float + float(A=float / float(A=@n1, B=10) {is=title K2Node_CallFunction}, B=1) {is=title K2Node_CommutativeAssociativeBinaryOperator}
    Max=1
   )
   > Set Pipe Scale(Pipe Scale:Vector=MakeVector(X=float + float(A=0, B=1), Y=@n2, Z=@n2)) -> (Output_Get:Vector)
   n3 = float * float(A=float * float(A=1, B=60), B=GetGameTimeInSeconds())
   n4 = Get bIsPlayingMeleeAnim()
   n5 = RandomFloatInRange(Min=float * float(A="Sin (Radians)"(A=@n3), B=SelectFloat(A=2, B=0.6, bPickA=@n4)), Max=1.7)
   > Set PipeRotate(PipeRotate:Rotator=MakeRotator(Pitch=@n5, Yaw=@n5)) -> (Output_Get:Rotator)
   n6 = GetGameTimeInSeconds()
   n7 = "Sin (Radians)"(A=float * float(A=float * float(A=1, B=@n6), B=0))
   n8 = SelectFloat(A=1, B=0.3, bPickA=Get bIsPlayingMeleeAnim())
   > Set Motor Rotate(Motor Rotate:Rotator=MakeRotator(Pitch=float * float(A=@n7, B=@n8))) -> (Output_Get:Rotator)
   n9 = GetGameTimeInSeconds()
   n10 = float * float(A=1, B=60)
   n11 = float * float(A="Sin (Radians)"(A=float * float(A=@n10, B=@n9)), B=SelectFloat(A=0.6, B=0.3, bPickA=Get bIsPlayingMeleeAnim()))
   > Set MainMassRotate(MainMassRotate:Rotator=MakeRotator(Pitch=@n11, Yaw=@n11)) -> (Output_Get:Rotator)
   n12 = Get bIsPlayingMeleeAnim()
   n13 = float * float(A=float * float(A=1, B=60), B=GetGameTimeInSeconds())
   n14 = RandomFloatInRange(Min=float * float(A="Sin (Radians)"(A=@n13), B=SelectFloat(A=2, B=1, bPickA=@n12)), Max=1.7)
   > Set PipeRotate_Small(PipeRotate_Small:Rotator=MakeRotator(Pitch=@n14, Yaw=@n14)) -> (Output_Get:Rotator)
   TryGetPawnOwner()
  ]}
]
variables [
 ExhaustPipeBig {type=number}
 ExhaustPipeBigRight {type=Vector}
 ExhaustPipeBigVector {type=Vector}
 ExhaustPipeSmall {type=number}
 ExhaustPipeSmall Right {type=Vector}
 ExhaustPipeSmallVector {type=Vector}
 MainMassRotate {type=Rotator}
 Motor Rotate {type=Rotator}
 Pipe Scale {type=Vector}
 PipeRotate {type=Rotator}
 PipeRotate_Small {type=Rotator}
]
